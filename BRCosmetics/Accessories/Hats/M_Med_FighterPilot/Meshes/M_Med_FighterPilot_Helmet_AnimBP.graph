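# Animation Blueprint for the fighter pilot helmet cosmetic that adds secondary motion to its hose.
# On initialization it caches the owner's skeletal mesh component, and the AnimGraph copies that mesh's pose as its base.
# Three chained Anim Dynamics nodes then simulate dyn_hose, dyn_hose_04 and dyn_hose_05 in actor space, using damped angular springs with tight angular limits.
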AnimBlueprint /BRCosmetics/Accessories/Hats/M_Med_FighterPilot/Meshes/M_Med_FighterPilot_Helmet_AnimBP
target_skeleton=&/M_Med_FighterPilot_Helmet_Skeleton
graphs [
 AnimGraph {nodes [
   n1 = Get MeshToCopy() -> (MeshToCopy:Skeletal Mesh Component Object Reference)
   n2 = Anim Dynamics - Bone: dyn_hose(ComponentPose=Local To Component(LocalPose=Copy Pose From Mesh(SourceMeshComponent=@n1)), Alpha=1, ExternalForce="0,0,0") {
    angularDampingOverride=0.992
    angularSpringConstant=500
    bAngularSpring=true
    bGravityOverrideInSimSpace=true
    bOverrideAngularDamping=true
    simulationSpace=Actor
    alphaBoolBlend {blendOption=Linear}
    boundBone {boneName=dyn_hose}
    physicsBodyDefinitions [{
      sphereCollisionRadius=0
      boundBone {boneName=dyn_hose}
      boxExtents=<2 20 20>
      constraintSetup {angularLimitsMax=<10 2 0>, angularLimitsMin=<-10 -2 0>, angularTarget=<0 0 0>}
      localJointOffset=<-5 0 0>
     }]
   }
   n3 = Anim Dynamics - Bone: dyn_hose_04(ComponentPose=@n2, Alpha=1, ExternalForce="0,0,0") {
    angularDampingOverride=0.992
    angularSpringConstant=500
    bAngularSpring=true
    bGravityOverrideInSimSpace=true
    bOverrideAngularDamping=true
    simulationSpace=Actor
    alphaBoolBlend {blendOption=Linear}
    boundBone {boneName=dyn_hose_04}
    physicsBodyDefinitions [{
      sphereCollisionRadius=0
      boundBone {boneName=dyn_hose_04}
      boxExtents=<2 20 20>
      constraintSetup {angularLimitsMax=<10 2 0>, angularLimitsMin=<-10 -2 0>, angularTarget=<0 0 0>}
      localJointOffset=<-5 0 0>
     }]
   }
   n4 = Anim Dynamics - Bone: dyn_hose_05(ComponentPose=@n3, Alpha=1, ExternalForce="0,0,0") {
    angularDampingOverride=0.992
    angularSpringConstant=500
    bAngularSpring=true
    bGravityOverrideInSimSpace=true
    bOverrideAngularDamping=true
    simulationSpace=Actor
    alphaBoolBlend {blendOption=Linear}
    boundBone {boneName=dyn_hose_05}
    physicsBodyDefinitions [{
      sphereCollisionRadius=0
      boundBone {boneName=dyn_hose_05}
      boxExtents=<2 20 20>
      constraintSetup {angularLimitsMax=<10 2 0>, angularLimitsMin=<-10 -2 0>, angularTarget=<0 0 0>}
      localJointOffset=<-5 0 0>
     }]
   }
   Output Pose(Result=Component To Local(ComponentPose=@n4)) {layerGroup=None}
  ]}
 EventGraph {nodes [
   Event BlueprintInitializeAnimation()
   > Set MeshToCopy(MeshToCopy:Skeletal Mesh Component Object Reference=GetOwnerSkeletalMeshComponent()) -> (Output_Get:Skeletal Mesh Component Object Reference)
   Event BlueprintUpdateAnimation()
  ]}
]
variables [MeshToCopy {type=SkeletalMeshComponent}]
